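# Animation Blueprint for the NinjaWarrior pickaxe that adds procedural secondary motion to its brush and shuriken bones.
# Stiff handle-relative AnimDynamics springs drive the brush; softer root-relative springs, including a three-bone chain, drive the shuriken.
# Trail controllers on both tips follow, then an additive pitch rotates the shuriken base.
# The EventGraph sets simulation-space and LOD variables from front-end and LOD state, but the AnimGraph reads neither.
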AnimBlueprint /BRCosmetics/Weapons/FORT_Melee/Pickaxe_NinjaWarrior/Meshes/NinjaWarrior_Axe_AnimBP
target_skeleton=&/NinjaWarrior_Axe_Skeleton
graphs [
 AnimGraph {nodes [
   AnimGraphNode_SubInput_0 = Input Pose()
   n1 = Anim Dynamics - Bone: Dyn_brush_up1!(
    SimulationSpace:AnimPhysSimSpaceType Enum=BoneRelative
    ComponentPose=Local To Component(LocalPose=@AnimGraphNode_SubInput_0)
    Alpha=1
    ExternalForce="0,0,0"
   ) -> (Pose) {
    angularDampingOverride=1
    angularSpringConstant=3500
    bAngularSpring=true
    bOverrideAngularDamping=true
    bUseGravityOverride=true
    simulationSpace=BoneRelative
    alphaBoolBlend {blendOption=Linear}
    boundBone {boneName=Dyn_brush_up1}
    componentAppliedLinearAccClamp=<3000 3000 3000>
    componentLinearAccScale=<0.3 0.3 0.3>
    componentLinearVelScale=<0.3 0.3 0.3>
    gravityOverride {z=-980}
    physicsBodyDefinitions [{
      boundBone {boneName=Dyn_brush_up1}
      boxExtents=<15 15 15>
      constraintSetup {angularLimitsMax=<10 10 10>, angularLimitsMin=<-10 -10 -10>, angularTarget=<1 0 0>}
      localJointOffset=<0.7 0 0>
     }]
    relativeSpaceBone {boneName=Melee_TwoHanded_Handle}
   }
   n2 = Anim Dynamics - Bone: Dyn_brush_up2!(SimulationSpace:AnimPhysSimSpaceType Enum=BoneRelative, ComponentPose=@n1, Alpha=1, ExternalForce="0,0,0") -> (Pose) {
    angularDampingOverride=1
    angularSpringConstant=3500
    bAngularSpring=true
    bOverrideAngularDamping=true
    bUseGravityOverride=true
    simulationSpace=BoneRelative
    alphaBoolBlend {blendOption=Linear}
    boundBone {boneName=Dyn_brush_up2}
    componentAppliedLinearAccClamp=<3000 3000 3000>
    componentLinearAccScale=<0.3 0.3 0.3>
    componentLinearVelScale=<0.3 0.3 0.3>
    gravityOverride {z=-980}
    physicsBodyDefinitions [{
      boundBone {boneName=Dyn_brush_up2}
      boxExtents=<15 15 15>
      constraintSetup {angularLimitsMax=<8 8 8>, angularLimitsMin=<-8 -7 -8>, angularTarget=<1 0 0>}
      localJointOffset=<0.7 0 0>
     }]
    relativeSpaceBone {boneName=Melee_TwoHanded_Handle}
   }
   n3 = Anim Dynamics - Bone: Dyn_brush_up3!(SimulationSpace:AnimPhysSimSpaceType Enum=BoneRelative, ComponentPose=@n2, Alpha=1, ExternalForce="0,0,0") -> (Pose) {
    angularDampingOverride=1
    angularSpringConstant=3500
    bAngularSpring=true
    bOverrideAngularDamping=true
    bUseGravityOverride=true
    simulationSpace=BoneRelative
    alphaBoolBlend {blendOption=Linear}
    boundBone {boneName=Dyn_brush_up3}
    componentAppliedLinearAccClamp=<3000 3000 3000>
    componentLinearAccScale=<0.3 0.3 0.3>
    componentLinearVelScale=<0.3 0.3 0.3>
    gravityOverride {z=-980}
    physicsBodyDefinitions [{
      boundBone {boneName=Dyn_brush_up3}
      boxExtents=<15 15 15>
      constraintSetup {angularLimitsMax=<7 7 5>, angularLimitsMin=<-7 -5 -5>, angularTarget=<1 0 0>}
      localJointOffset=<0.7 0 0>
     }]
    relativeSpaceBone {boneName=Melee_TwoHanded_Handle}
   }
   n4 = Anim Dynamics - Bone: Dyn_brush_up4!(SimulationSpace:AnimPhysSimSpaceType Enum=BoneRelative, ComponentPose=@n3, Alpha=1, ExternalForce="0,0,0") -> (Pose) {
    angularDampingOverride=1
    angularSpringConstant=3500
    bAngularSpring=true
    bOverrideAngularDamping=true
    bUseGravityOverride=true
    simulationSpace=BoneRelative
    alphaBoolBlend {blendOption=Linear}
    boundBone {boneName=Dyn_brush_up4}
    componentAppliedLinearAccClamp=<3000 3000 3000>
    componentLinearAccScale=<0.3 0.3 0.3>
    componentLinearVelScale=<0.3 0.3 0.3>
    gravityOverride {z=-980}
    physicsBodyDefinitions [{
      boundBone {boneName=Dyn_brush_up4}
      boxExtents=<15 15 15>
      constraintSetup {angularLimitsMax=<5 5 5>, angularLimitsMin=<-5 -5 -5>, angularTarget=<1 0 0>}
      localJointOffset=<0.65 0 0>
     }]
    relativeSpaceBone {boneName=Melee_TwoHanded_Handle}
   }
   n5 = Anim Dynamics - Bone: Dyn_brush_up5!(SimulationSpace:AnimPhysSimSpaceType Enum=BoneRelative, ComponentPose=@n4, Alpha=1, ExternalForce="0,0,0") -> (Pose) {
    angularDampingOverride=1
    angularSpringConstant=3500
    bAngularSpring=true
    bOverrideAngularDamping=true
    bUseGravityOverride=true
    simulationSpace=BoneRelative
    alphaBoolBlend {blendOption=Linear}
    boundBone {boneName=Dyn_brush_up5}
    componentAppliedLinearAccClamp=<3000 3000 3000>
    componentLinearAccScale=<0.3 0.3 0.3>
    componentLinearVelScale=<0.3 0.3 0.3>
    gravityOverride {z=-980}
    physicsBodyDefinitions [{
      boundBone {boneName=Dyn_brush_up5}
      boxExtents=<15 15 15>
      constraintSetup {angularLimitsMax=<3 3 3>, angularLimitsMin=<-3 -3 -3>, angularTarget=<1 0 0>}
      localJointOffset=<0.3 0 0>
     }]
    relativeSpaceBone {boneName=Melee_TwoHanded_Handle}
   }
   n6 = Anim Dynamics - Bone: Dyn_suriken_dwn1!(SimulationSpace:AnimPhysSimSpaceType Enum, ComponentPose=@n5, Alpha=1, ExternalForce="0,0,0") -> (Pose) {
    angularDampingOverride=0.999
    angularSpringConstant=250
    bAngularSpring=true
    bOverrideAngularDamping=true
    bUseGravityOverride=true
    simulationSpace=RootRelative
    alphaBoolBlend {blendOption=Linear}
    boundBone {boneName=Dyn_suriken_dwn1}
    componentAppliedLinearAccClamp=<5000 5000 5000>
    componentLinearAccScale=<0.3 0.3 0.3>
    componentLinearVelScale=<0.3 0.3 0.3>
    gravityOverride {z=-980}
    physicsBodyDefinitions [{
      boundBone {boneName=Dyn_suriken_dwn1}
      boxExtents=<10 10 10>
      constraintSetup {angularLimitsMax=<5 85 30>, angularLimitsMin=<-5 -30 -29>, angularTarget=<0 0 0>}
      localJointOffset=<2 0 0>
     }]
   }
   n7 = Anim Dynamics - Bone: Dyn_suriken_dwn2!(SimulationSpace:AnimPhysSimSpaceType Enum, ComponentPose=@n6, Alpha=1, ExternalForce="0,0,0") -> (Pose) {
    angularDampingOverride=0.999
    angularSpringConstant=250
    bAngularSpring=true
    bOverrideAngularDamping=true
    bUseGravityOverride=true
    simulationSpace=RootRelative
    alphaBoolBlend {blendOption=Linear}
    boundBone {boneName=Dyn_suriken_dwn2}
    componentAppliedLinearAccClamp=<5000 5000 5000>
    componentLinearAccScale=<0.3 0.3 0.3>
    componentLinearVelScale=<0.3 0.3 0.3>
    gravityOverride {z=-980}
    physicsBodyDefinitions [{
      boundBone {boneName=Dyn_suriken_dwn2}
      boxExtents=<13 13 13>
      constraintSetup {angularLimitsMax=<5 85 30>, angularLimitsMin=<-5 -45 -29>, angularTarget=<0 0 0>}
      localJointOffset=<1 0 0>
     }]
   }
   n8 = "Anim Dynamics - Chain: Dyn_suriken_dwn3 -> Dyn_suriken_dwn5"!(SimulationSpace:AnimPhysSimSpaceType Enum, ComponentPose=@n7, Alpha=1, ExternalForce="0,0,0") -> (Pose) {
    angularDampingOverride=0.999
    angularSpringConstant=500
    bAngularSpring=true
    bChain=true
    bOverrideAngularDamping=true
    bUseGravityOverride=true
    simulationSpace=RootRelative
    alphaBoolBlend {blendOption=Linear}
    boundBone {boneName=Dyn_suriken_dwn3}
    chainEnd {boneName=Dyn_suriken_dwn5}
    componentAppliedLinearAccClamp=<5000 5000 5000>
    componentLinearAccScale=<0.3 0.3 0.3>
    componentLinearVelScale=<0.3 0.3 0.3>
    gravityOverride {z=-980}
    physicsBodyDefinitions [
     {
      boundBone {boneName=Dyn_suriken_dwn3}
      boxExtents=<13 13 13>
      constraintSetup {angularLimitsMax=<5 30 30>, angularLimitsMin=<-5 -45 -29>, angularTarget=<0 0 0>}
      localJointOffset=<0.7 0 0>
     }
     {
      boundBone {boneName=Dyn_suriken_dwn4}
      boxExtents=<10 10 10>
      constraintSetup {angularLimitsMax=<5 85 30>, angularLimitsMin=<-5 -45 -29>, angularTarget=<0 0 0>}
      localJointOffset=<1 0 0>
     }
     {
      boundBone {boneName=Dyn_suriken_dwn5}
      boxExtents=<10 10 10>
      constraintSetup {angularLimitsMax=<5 85 30>, angularLimitsMin=<-5 -45 -29>, angularTarget=<0 0 0>}
      localJointOffset=<1 0 0>
     }
    ]
   }
   n9 = Trail controller - Bone: Dyn_suriken_dwn5!(ComponentPose=@n8, bAlphaBoolEnabled:bool=True) -> (Pose) {
    alphaInputType=Bool
    bLimitRotation=true
    bShowTrailLocation=true
    chainLength=5
    maxDeltaTime=0.33
    relaxationSpeedScale=2
    fakeVelocity {y=1}
    rotationLimits [
     {limitMax=<5 85 30>, limitMin=<-5 -30 -30>}
     {limitMax=<5 30 30>, limitMin=<-5 -30 -30>}
     {limitMax=<5 30 30>, limitMin=<-5 -30 -30>}
     {limitMax=<5 30 30>, limitMin=<-5 -30 -30>}
     {limitMax=<5 30 30>, limitMin=<-5 -30 -30>}
    ]
    rotationOffsets [<0 0 0>, <0 0 0>, <0 0 0>, <0 0 0>, <0 0 0>]
    trailBone {boneName=Dyn_suriken_dwn5}
    trailRelaxationSpeed {editorCurveData {keys [<at -0.0171 5.7275 RCIM_Linear>,<at 1.0232 9.5088 RCIM_Linear>]}}
   }
   n10 = Trail controller - Bone: Dyn_brush_up5!(ComponentPose=@n9, bAlphaBoolEnabled:bool=True) -> (Pose) {
    alphaInputType=Bool
    bLimitRotation=true
    bShowTrailLocation=true
    chainLength=4
    maxDeltaTime=0.33
    relaxationSpeedScale=5
    baseJoint {boneName=Base_suriken_dwn}
    fakeVelocity {y=1}
    rotationLimits [
     {limitMax=<5 85 30>, limitMin=<-5 -30 -30>}
     {limitMax=<5 30 30>, limitMin=<-5 -30 -30>}
     {limitMax=<5 30 30>, limitMin=<-5 -30 -30>}
     {limitMax=<5 30 30>, limitMin=<-5 -30 -30>}
    ]
    rotationOffsets [<0 0 0>, <0 0 0>, <0 0 0>, <0 0 0>]
    trailBone {boneName=Dyn_brush_up5}
   }
   n11 = "Transform (Modify) Bone - Bone: Base_suriken_dwn"!(ComponentPose=@n10, bAlphaBoolEnabled:bool=True, Translation="0,0,0", Rotation="-75.000,0.000,-0.000", Scale="1,1,1") -> (Pose) {
    alphaInputType=Bool
    rotationMode=BMM_Additive
    translationSpace=BCS_BoneSpace
    alphaBoolBlend {blendOption=Linear}
    boneToModify {boneName=Base_suriken_dwn}
    rotation {pitch=-75}
   }
   Output Pose(Result=Component To Local(ComponentPose=@n11)) {layerGroup=None}
  ]}
 EventGraph {nodes [
   Event BlueprintUpdateAnimation()
   > n1 = Sequence()
   > n2 = Branch(Condition=Get bIsInFrontEnd())
   > Set SimlationSpace(SimlationSpace:AnimPhysSimSpaceType Enum=World) -> (Output_Get:AnimPhysSimSpaceType Enum)
   Set SimlationSpace(execute=@n2.else, SimlationSpace=Component)
   n3 = Branch(execute=@n1.then_1, Condition="integer >= integer"(A=Get CurrentLOD(), B=2))
   > Set Lod 2 and above(Lod 2 and above:bool=true) -> (Output_Get:bool)
   Set Lod 2 and above(execute=@n3.else, Lod 2 and above=false)
   TryGetPawnOwner()
  ]}
]
variables [
 Lod 2 and above {type=boolean}
 SimlationSpace {type=AnimPhysSimSpaceType}
]
